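# Physics asset for the Obsidian Items gadget skeletal mesh, a cosmetic animation prop.
# It has three capsule bodies, one on the main obsidian bone and one on each of two rock bones.
# Each rock body is constrained directly to the main body with 45-degree swing and twist limit angles, forming a simple hub with two branches.
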PhysicsAsset /BRCosmetics/Animation/Props/Obsidian_Items/Meshes/Obsidian_Items_Gadget_PhysicsAsset
bodies {
 JNT_Obsidian_Main [{shapeType=Capsule, length=13.9903, radius=10.9506, center=<-1.647 0.6345 0.6435>, rotation=<rot 20.6418 -90 29.6257>}]
 JNT_rock_15 [{shapeType=Capsule, length=16.2797, radius=11.707, center=<-0.1376 1.1229 1.6088>, rotation=<rot 0 141.5234 -81.0414>}]
 JNT_rock_17 [{shapeType=Capsule, length=18.7955, radius=15.436, center=<0.282 -0.0198 0.3545>, rotation=<rot 0 -76.8615 -89.988>}]
}
constraints [
 {bone1Name=JNT_rock_15, bone2Name=JNT_Obsidian_Main, swing1LimitDegrees=45, swing2LimitDegrees=45, twistLimitDegrees=45}
 {bone1Name=JNT_rock_17, bone2Name=JNT_Obsidian_Main, swing1LimitDegrees=45, swing2LimitDegrees=45, twistLimitDegrees=45}
]
